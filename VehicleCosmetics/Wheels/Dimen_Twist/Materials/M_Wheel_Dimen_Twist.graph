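# Masked, skeletal-mesh material for the Dimen Twist vehicle cosmetic wheel, with an animated marble look.
# Two panning marble noise layers, one bump-offset for fake depth, combine with fresnel terms to drive base color and emissive on a non-metallic, fully rough surface.
# The Marble Only switch selects pure marble or a blend with rim and tire masks and normals.
# Vertex-color-gated opacity masking precedes vehicle damage and de-rez transition functions.
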Material /VehicleCosmetics/Wheels/Dimen_Twist/Materials/M_Wheel_Dimen_Twist
graph {
 properties {
  bAutomaticallySetUsageInEditor=false
  bUseMaterialAttributes=true
  bUsedWithMeshDeformer=true
  bUsedWithSkeletalMesh=true
  blendMode=BLEND_Masked
  previewMesh=/VehicleCosmetics/Wheels/Dimen_Twist/Meshes/SK_Wheel_DimenTwist
  usedShadingModels=Default Lit
 }
 nodes [
  n1 = Constant()
  n2 = TextureSampleParameter2D() {parameterName=OP Masks, samplerType=SAMPLERTYPE_LinearColor, texture=/VehicleCosmetics/Wheels/Dimen_Twist/Textures/T_Dimen_Twist_Rim_OP_LOD}
  n3 = SetMaterialAttributes!(
   MaterialAttributes
   Base Color=NamedRerouteUsage()
   Emissive Color=NamedRerouteUsage()
   Normal=NamedRerouteUsage()
   Metallic=@n1
   Specular=@n1
   Roughness=Constant() {r=1}
   Opacity Mask=LinearInterpolate(B=@n2.G, Alpha=VertexColor().B) {constA=1, constB=0}
  ) {attributeSetTypes [
    #1
    #2
    #3
    #4
    #5
    #6
    #7
   ]}
  MaterialFunctionCall(In=MaterialFunctionCall(Material=@n3) {materialFunction=&MF_BaseVehicle_Damage}.Result) {materialFunction=&MF_VehicleDeRez_Transition}
  NamedRerouteDeclaration(None=VertexColor().R) {name=VC_R, nodeColor=<rgba 1 0.5625 0 1>}
  n4 = TextureSampleParameter2D() {
   parameterName=Marble Normal
   samplerType=SAMPLERTYPE_Normal
   texture=/VehicleCosmetics/Wheels/Dimen_Twist/Textures/T_Dimen_Twist_Tire_Marble_N
  }
  n5 = TextureCoordinate() {uTiling=3, vTiling=5}
  n6 = ScalarParameter() {defaultValue=2, parameterName=Marble Tile}
  n7 = Add(
   A=StaticSwitchParameter(True=TextureCoordinate() {uTiling=2, vTiling=1.5}, False=Multiply(A=@n5, B=@n6)) {defaultValue=true, parameterName=Marble Only}
   B=Constant2Vector() {g=0.15}
  )
  n8 = Multiply(A=ScalarParameter() {defaultValue=0.5, parameterName=Pan Speed}, B=Time())
  n9 = Panner(Coordinate=@n7, Time=@n8) {speedX=0.05}
  n10 = TextureSampleParameter2D(UVs=@n9) {
   parameterName=Marble Normal Noise
   samplerType=SAMPLERTYPE_Normal
   texture=&T_Dimen_Twist_Tire_Marble_N
  }
  n11 = AppendVector(A=Multiply(A=ComponentMask(None=@n10.RGB) {g=true, r=true}, B=ScalarParameter() {defaultValue=0.5, parameterName=Normal Strength}), B=@n10.B)
  n12 = MaterialFunctionCall(BaseNormal=@n4.RGB, AdditionalNormal=@n11) {materialFunction=&BlendAngleCorrectedNormals}
  n13 = TextureSampleParameter2D() {parameterName=Rim Normal, samplerType=SAMPLERTYPE_Normal, texture=/VehicleCosmetics/Wheels/Dimen_Twist/Textures/T_Dimen_Twist_Rim_N}
  n14 = TextureSampleParameter2D() {parameterName=Tire Normal, samplerType=SAMPLERTYPE_Normal, texture=/VehicleCosmetics/Wheels/Dimen_Twist/Textures/T_Dimen_Twist_Tire_N}
  n15 = MaterialFunctionCall(BaseNormal=LinearInterpolate(A=@n13.RGB, B=@n14.RGB, Alpha=NamedRerouteUsage()), AdditionalNormal=@n11) {materialFunction=&BlendAngleCorrectedNormals}
  NamedRerouteDeclaration(None=StaticSwitchParameter(True=@n12.Result, False=@n15.Result) {defaultValue=true, parameterName=Marble Only}) {name=Normal, nodeColor=<rgba 0.5 0.5 1 1>}
  n16 = VectorParameter() {parameterName=PaintColor, defaultValue=<rgba 0.8308 0.8308 0.8308 1>}
  NamedRerouteDeclaration(None=@n16.RGB) {name=Paint Color, nodeColor=<rgba 0 0.8828 1 1>}
  NamedRerouteDeclaration(None=@n16.A) {name=Paint Color A, nodeColor=<rgba 1 0.7969 0 1>}
  n17 = MaterialFunctionCall!(Normal, Exponent=Constant() {r=5}) {materialFunction=/VehicleCosmetics/Wheels/Cola_Holographic/Materials/MF_Fresnel_Classic}
  n18 = Clamp(None=ConstantBiasScale(None=Clamp(None=@n17.Result)) {bias=0.1, scale=1})
  n19 = Transform(None=CameraVectorWS()) {transformSourceType=TRANSFORMSOURCE_World, transformType=TRANSFORM_Tangent}
  n20 = Clamp(None=Power(Base=ComponentMask(None=@n19) {b=true}) {constExponent=15})
  n21 = Add(A=@n18, B=Clamp(None=ConstantBiasScale(None=@n20) {bias=0, scale=2}) {maxDefault=2})
  n22 = NamedRerouteUsage()
  NamedRerouteDeclaration(None=LinearInterpolate(A=Multiply(A=@n21, B=@n22), B=NamedRerouteUsage(), Alpha=NamedRerouteUsage())) {name=Color Fresnel 1, nodeColor=<rgba 0.5234 0 1 1>}
  n23 = ConstantBiasScale(None=@n21) {bias=0.12, scale=1}
  n24 = NamedRerouteUsage()
  NamedRerouteDeclaration(None=LinearInterpolate(A=Multiply(A=@n23, B=@n24), B=@n24, Alpha=NamedRerouteUsage())) {name=Color Fresnel 2, nodeColor=<rgba 0.5234 0 1 1>}
  n25 = TextureSample(UVs=Panner(Coordinate=TextureCoordinate()) {speedX=0.25}) {samplerSource=SSM_Wrap_WorldGroupSettings, texture=/VehicleCosmetics/Wheels/_SharedTextures/FX/SharpGradient_01_Heightmap}
  NamedRerouteDeclaration(None=Clamp(None=ConstantBiasScale(None=@n25.R) {bias=0.1, scale=1.5}) {maxDefault=1.5, minDefault=0.1}) {name=Grad Panner, nodeColor=<rgba 1 0 0.1172 1>}
  n26 = ScalarParameter() {defaultValue=4, parameterName=Emissive Power}
  NamedRerouteDeclaration(None=@n26) {name=Emissive Power, nodeColor=<rgba 0 0.4375 1 1>}
  n27 = TextureSampleParameter2D() {parameterName=Rim Masks, samplerType=SAMPLERTYPE_LinearColor, texture=/VehicleCosmetics/Wheels/Dimen_Twist/Textures/T_Dimen_Twist_Rim_M}
  n28 = TextureSampleParameter2D() {parameterName=Tire Masks, samplerType=SAMPLERTYPE_LinearColor, texture=/VehicleCosmetics/Wheels/Dimen_Twist/Textures/T_Dimen_Twist_Rubber_M}
  NamedRerouteDeclaration(None=LinearInterpolate(A=@n27.R, B=@n28.R, Alpha=NamedRerouteUsage())) {name=Masks Lerp R, nodeColor=<rgba 0.7578 1 0 1>}
  n29 = TextureObjectParameter() {parameterName=Noise, samplerType=SAMPLERTYPE_LinearColor, texture=/VehicleCosmetics/Wheels/Dimen_Twist/Textures/T_Dimen_Twist_Tire_Marble_M}
  n30 = TextureSample(UVs=@n9, Tex=@n29) {
   samplerSource=SSM_Wrap_WorldGroupSettings
   samplerType=SAMPLERTYPE_LinearColor
   texture=&T_Dimen_Twist_Tire_Marble_M
  }
  n31 = Multiply(A=@n7) {constB=2}
  n32 = Constant() {r=-2}
  n33 = TextureSample(UVs=Panner(Coordinate=BumpOffset(Coordinate=@n31, Height=@n32), Time=@n8) {speedX=0.05}, Tex=@n29) {
   samplerSource=SSM_Wrap_WorldGroupSettings
   samplerType=SAMPLERTYPE_LinearColor
   texture=&T_Dimen_Twist_Tire_Marble_M
  }
  n34 = Max(A=Clamp(None=ConstantBiasScale(None=@n30.R) {bias=-0.01, scale=1}), B=ConstantBiasScale(None=@n33.R) {bias=0, scale=0.05})
  n35 = StaticSwitchParameter(True=@n12.Result, False=@n15.Result) {defaultValue=true, parameterName=Marble Only}
  n36 = Constant() {r=1}
  n37 = ConstantBiasScale(None=OneMinus(None=MaterialFunctionCall(Normal=@n35, Exponent=@n36) {materialFunction=&MF_Fresnel_Classic}.Result)) {bias=-0.85, scale=5}
  n38 = Clamp(None=Subtract(A=Multiply(A=@n34, B=@n26), B=Clamp(None=@n37) {maxDefault=0.3}))
  n39 = NamedRerouteUsage()
  n40 = OneMinus(None=@n34)
  n41 = Clamp(None=ConstantBiasScale(None=Power(Base=@n40) {constExponent=50}) {bias=0})
  n42 = ConstantBiasScale(None=@n41) {bias=0, scale=5}
  n43 = LinearInterpolate(A=Clamp(None=Subtract(A=@n27.R, B=@n27.B)), Alpha=NamedRerouteUsage()) {constB=0}
  n44 = LinearInterpolate(A=Add(A=@n27.B, B=ScalarParameter() {parameterName=Spokes Minimum Color}), B=@n28.B, Alpha=NamedRerouteUsage())
  n45 = NamedRerouteUsage()
  n46 = Max(
   A=Multiply(A=NamedRerouteUsage(), B=LinearInterpolate(A=@n41, B=@n42, Alpha=@n43))
   B=Multiply(A=Multiply(A=@n44, B=@n45), B=NamedRerouteUsage())
  )
  n47 = NamedRerouteUsage()
  NamedRerouteDeclaration(None=StaticSwitchParameter(True=Multiply(A=@n38, B=@n39), False=Multiply(A=@n46, B=@n47)) {defaultValue=true, parameterName=Marble Only}) {name=Final Emissive, nodeColor=<rgba 0.9219 0 1 1>}
  NamedRerouteDeclaration(None=LinearInterpolate(A=@n27.G, B=@n28.G, Alpha=NamedRerouteUsage())) {name=Masks Lerp G, nodeColor=<rgba 0.7578 1 0 1>}
  n48 = Subtract(A=Clamp(None=@n37), B=@n34)
  n49 = ConstantBiasScale(None=MaterialFunctionCall(Normal=@n35, Exponent=Constant() {r=6}) {materialFunction=&MF_Fresnel_Classic}.Result) {bias=-0.1, scale=5}
  n50 = LinearInterpolate(B=NamedRerouteUsage(), Alpha=Max(A=Clamp(None=@n48), B=Clamp(None=@n49))) {constA=0.02}
  n51 = NamedRerouteUsage()
  n52 = Transform(None=CameraVectorWS()) {transformSourceType=TRANSFORMSOURCE_World, transformType=TRANSFORM_Tangent}
  n53 = Clamp(None=Power(Base=ComponentMask(None=@n52) {b=true}) {constExponent=10})
  n54 = Add(A=@n18, B=Clamp(None=ConstantBiasScale(None=@n53) {bias=0, scale=2}) {maxDefault=1.5})
  n55 = Multiply(
   A=LinearInterpolate(A=LinearInterpolate(B=@n18, Alpha=@n51), B=Multiply(A=@n54) {constB=0.125}, Alpha=NamedRerouteUsage())
   B=NamedRerouteUsage()
  )
  n56 = NamedRerouteUsage()
  NamedRerouteDeclaration(None=StaticSwitchParameter(True=@n50, False=LinearInterpolate(A=@n55, B=@n50, Alpha=@n56)) {defaultValue=true, parameterName=Marble Only}) {name=BC, nodeColor=<rgba 1 0.2812 0 1>}
 ]
}
